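; List view data object for the EOB Pass in the HS Season art tool, holding the data for one list entry.
; A custom Init event packs an ID name, Fort release version, item name, description, gameplay tags, asset path, series definition and ShotGrid ID into an EOB ID struct.
; It stores that struct in its DataListStructure variable and contains no other logic.
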(event Custom|Init
  (params (ID%20Name :type Text) (Item%20Fort%20Release%20Version :type Fort%20Release%20Version%20Structure) (Item%20Name :type Text) (Item%20Description :type Text) (Item%20Tags :type Gameplay%20Tag%20Container%20Structure) (AssetPath :type String) (SeriesItem :type Fort%20Item%20Series%20Definition%20Object%20Reference) (IDShotGrid :type Integer))
  (Variables|Default|SetDataListStructure (Utilities|Struct|MakeEOBIDStructure ID%20Name Item%20Fort%20Release%20Version Item%20Name Item%20Description Item%20Tags AssetPath SeriesItem IDShotGrid)))
