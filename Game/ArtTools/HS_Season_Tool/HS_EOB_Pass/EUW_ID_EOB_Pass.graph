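# Editor utility list-entry widget for the EOB Pass art tool, showing one Fortnite item ID with Reload, Load ID and Remove buttons.
# Setting the list object caches its data; Load ID (duplicated in LoadID_Event) copies those fields into the parent EUW_EOB_Pass form.
# An invisible button over the name runs inline Python to open the asset.
# Reload re-reads the FortItemDefinition; Remove drops the entry from the parent array and ListView.
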EditorUtilityWidgetBlueprint /Game/ArtTools/HS_Season_Tool/HS_EOB_Pass/EUW_ID_EOB_Pass
graphs [
 EventGraph {nodes [
   "On Clicked (BTN_RemoveFromList)"() {is=title K2Node_ComponentBoundEvent}
   > Remove Item!(
    execute
    TargetArray:Array of EOB Pass List View Object Object References=Get ArrayOfObjectsOnList!(self:EUW EOB Pass Object Reference=Get ParentMenu() -> (ParentMenu:EUW EOB Pass Object Reference)) -> (ArrayOfObjectsOnList:Array of EOB Pass List View Object Object References)
    Item:"EOB Pass List View Object Object Reference (by ref)"=Get As EOB Pass List View Object() -> (As EOB Pass List View Object:EOB Pass List View Object Object Reference)
   ) -> (then, ReturnValue:bool)
   > n1 = Cast To ListView!(
    execute
    Object=GetOwningListView!(UserListEntry:User List Entry Interface=Self-Reference()) -> (ReturnValue:List View Base Object Reference)
   ) -> (then, CastFailed, AsList View:List View Object Reference) {is=title K2Node_DynamicCast}
   > RemoveItem!(execute, self:List View Object Reference=@n1, Item:Object Reference=Get List Item Object() -> (List Item Object:Object Reference)) -> (then)
   "On Clicked (BTN_LoadID)"() {is=title K2Node_ComponentBoundEvent}
   > n2 = Cast To EUW_EOB_Pass!(execute, Object=Get ParentMenu()) -> (then, CastFailed, AsEUW EOB Pass:EUW EOB Pass Object Reference) {is=title K2Node_DynamicCast}
   > Set As EUW EOB Pass(As EUW EOB Pass:EUW EOB Pass Object Reference=@n2) -> (Output_Get:EUW EOB Pass Object Reference)
   > n3 = Cast To EOB_Pass_ListView_Object!(execute, Object=Get List Item Object()) -> (then, CastFailed, AsEOB Pass List View Object:EOB Pass List View Object Object Reference) {is=title K2Node_DynamicCast}
   > n4 = Set CurrentObjectListed!(
    execute
    CurrentObjectListed:EOB Pass List View Object Object Reference=@n3
    self:EUW EOB Pass Object Reference=Get As EUW EOB Pass() -> (As EUW EOB Pass:EUW EOB Pass Object Reference)
   ) -> (then, Output_Get:EOB Pass List View Object Object Reference)
   n5 = Break EOB ID Structure!(EOB_ID_Structure:"EOB ID Structure Structure (by ref)"=Get DataListStructure!(self:EOB Pass List View Object Object Reference=@n4) -> (DataListStructure:EOB ID Structure Structure)) -> (
    IDName_2_#1:Text
    ItemFortReleaseVersion_11_#2:Fort Release Version Structure
    ItemName_4_#3:Text
    ItemDescription_6_#4:Text
    ItemTags_13_#5:Gameplay Tag Container Structure
    AssetPath_16_#6:String
    SeriesItem_20_#7:Fort Item Series Definition Object Reference
    IDShotGrid_23_#8:int
   ) {is=title K2Node_BreakStruct}
   > Set Text(Text=@n5, self=Get ID_Name_Loaded!(self:EUW EOB Pass Object Reference=Get As EUW EOB Pass()) -> (ID_Name_Loaded:Text Object Reference))
   > Set FortReleaseVersion!(
    execute
    FortReleaseVersion:Fort Release Version Structure=@n5.ItemFortReleaseVersion_11_#2
    self:EUW EOB Pass Object Reference=Get As EUW EOB Pass()
   ) -> (then, Output_Get:Fort Release Version Structure)
   > Set Text(
    Text=@n5.ItemName_4_#3
    self:Editable Text Object Reference=Get ItemName_Input!(self:EUW EOB Pass Object Reference=Get As EUW EOB Pass()) -> (ItemName_Input:Editable Text Object Reference)
   )
   > Set Text(
    Text=@n5.ItemDescription_6_#4
    self=Get ItemDescription_Input!(self:EUW EOB Pass Object Reference=Get As EUW EOB Pass()) -> (ItemDescription_Input:Editable Text Object Reference)
   )
   > Set GameplayTags!(
    execute
    GameplayTags:Gameplay Tag Container Structure=@n5.ItemTags_13_#5
    self:EUW EOB Pass Object Reference=Get As EUW EOB Pass()
   ) -> (then, Output_Get:Gameplay Tag Container Structure)
   > Set CurrentLoadedAssetPath!(
    execute
    CurrentLoadedAssetPath:String=@n5.AssetPath_16_#6
    self:EUW EOB Pass Object Reference=Get As EUW EOB Pass()
   ) -> (then, Output_Get:String)
   > Set SeriesTags!(
    execute
    SeriesTags:Fort Item Series Definition Object Reference=@n5.SeriesItem_20_#7
    self:EUW EOB Pass Object Reference=Get As EUW EOB Pass()
   ) -> (then, Output_Get:Fort Item Series Definition Object Reference)
   > Set Text(
    Text="To Text (Integer)"(Value=@n5.IDShotGrid_23_#8, bUseGrouping=false)
    self=Get ShotgridID_Input!(self:EUW EOB Pass Object Reference=Get As EUW EOB Pass()) -> (ShotgridID_Input:Editable Text Object Reference)
   )
   > SetCurrentFortItemDefinition(self:EUW EOB Pass Object Reference=Get As EUW EOB Pass())
   LogString(execute=@n3.CastFailed, InString=ERROR)
   LogString(execute=@n2.CastFailed, InString=No cast)
   "On Clicked (btn_OpenID)"() {is=title K2Node_ComponentBoundEvent}
   > Execute Python Script!(
    execute
    PythonScript="import unreal\r\n\r\nasset_path = asset_path\r\nasset = unreal.EditorAssetLibrary.load_asset(asset_path)\r\n\r\nif asset:\r\n    unreal.AssetToolsHelpers.get_asset_tools().open_editor_for_assets([asset])"
    asset_path:String=Get Asset Path() -> (Asset Path:String)
   ) -> (then, ReturnValue) {inputs [asset_path]}
   "On Clicked (BTN_Reload)"() {is=title K2Node_ComponentBoundEvent}
   n6 = Get As EOB Pass List View Object()
   > n7 = LoadAsset(AssetPath=Break EOB ID Structure(EOB_ID_Structure=Get DataListStructure(self=@n6)).AssetPath_16_#6)
   > n8 = Cast To FortItemDefinition!(execute, Object=@n7) -> (then, CastFailed, AsFort Item Definition:Fort Item Definition Object Reference) {is=title K2Node_DynamicCast}
   > ReapplyDataToListObject!(
    execute
    self:EUW EOB Pass Object Reference=Get ParentMenu()
    ListViewObjectFortItemDefinition:Item Definition Base Object Reference=@n8
    Object Target:EOB Pass List View Object Object Reference=Get As EOB Pass List View Object()
   ) -> (then)
   n9 = LoadID_Event!() -> (OutputDelegate, then, Object:Object Reference) {is=title K2Node_CustomEvent}
   > n10 = Cast To EUW_EOB_Pass(Object=Get ParentMenu())
   > Set As EUW EOB Pass(As EUW EOB Pass=@n10)
   > n11 = Cast To EOB_Pass_ListView_Object(Object=@n9.Object)
   > n12 = Set CurrentObjectListed(CurrentObjectListed=@n11, self=Get As EUW EOB Pass())
   n13 = Break EOB ID Structure(EOB_ID_Structure=Get DataListStructure(self=@n12))
   > Set Text(Text=@n13, self:Text Object Reference=Get ID_Name_Loaded(self=Get As EUW EOB Pass()))
   > Set FortReleaseVersion(FortReleaseVersion=@n13.ItemFortReleaseVersion_11_#2, self=Get As EUW EOB Pass())
   > Set Text(Text=@n13.ItemName_4_#3, self:Editable Text Object Reference=Get ItemName_Input(self=Get As EUW EOB Pass()))
   > Set Text(Text=@n13.ItemDescription_6_#4, self=Get ItemDescription_Input(self=Get As EUW EOB Pass()))
   > Set GameplayTags(GameplayTags=@n13.ItemTags_13_#5, self=Get As EUW EOB Pass())
   > Set CurrentLoadedAssetPath(CurrentLoadedAssetPath=@n13.AssetPath_16_#6, self=Get As EUW EOB Pass())
   > Set SeriesTags(SeriesTags=@n13.SeriesItem_20_#7, self=Get As EUW EOB Pass())
   > Set Text(
    Text="To Text (Integer)"(Value=@n13.IDShotGrid_23_#8, bUseGrouping=false)
    self=Get ShotgridID_Input(self=Get As EUW EOB Pass())
   )
   LogString(execute=@n10.CastFailed, InString=No cast)
   n14 = Event OnListItemObjectSet()
   > n15 = Set List Item Object(List Item Object:Object Reference=@n14.ListItemObject) -> (Output_Get:Object Reference)
   > n16 = Cast To EOB_Pass_ListView_Object(Object=@n15)
   > Set As EOB Pass List View Object(As EOB Pass List View Object:EOB Pass List View Object Object Reference=@n16) -> (Output_Get:EOB Pass List View Object Object Reference)
   n17 = Break EOB ID Structure(EOB_ID_Structure=Get DataListStructure(self=@n16))
   > Set Text(Text=@n17, self:Text Object Reference=Get ID_Name() -> (ID_Name:Text Object Reference))
   > Set Asset Path(Asset Path:String=@n17.AssetPath_16_#6) -> (Output_Get:String)
   Event PreConstruct()
   Event Construct()
   Event Tick()
  ]}
 OnLoadBTNCLicked {nodes [OnLoadBTNCLicked() {is=title K2Node_FunctionEntry}]}
]
variables [
 As EOB Pass List View Object {type=EOB_Pass_ListView_Object_C, default=null}
 As EUW EOB Pass {type=EUW_EOB_Pass_C, default=null}
 Asset Path {type=string, default=""}
 CurrentObject_EOB_Pass {type=EOB_Pass_ListView_Object_C, default=null}
 ItemDescription_ID {type=string, default=""}
 ItemFRV_ID {type=FortReleaseVersion, default {versionName=Unset}}
 ItemName_ID {type=string, default=""}
 ItemTags_ID {type=GameplayTagContainer, default {gameplayTags []}}
 List Item Object {type=Object, default=null}
 OnLoadBTNCLicked {default=()}
 ParentMenu {type=EUW_EOB_Pass_C, default=null}
]
widgets {
 info {rootWidgetClass=&StackBox, widgetCount=18}
 widgets [
  StackBox_0 {widgetClassPath=&StackBox, widget=StackBox_0}
  Overlay_0 {
   widgetClassPath=&Overlay
   parent=StackBox_0
   widget=Overlay_0
   slot {class=StackBoxSlot, properties {verticalAlignment=VAlign_Center, padding=<ltrb 4 4 4 4>, size {sizeRule=Fill, value=1}}}
  }
  Image_97 {
   widgetClassPath=&Image
   bIsVariable=true
   parent=Overlay_0
   widget=Image_97
   properties {colorAndOpacity=<rgba 0.0677 0.0677 0.0677 1>}
   slot {class=OverlaySlot, properties {horizontalAlignment=HAlign_Fill, verticalAlignment=VAlign_Fill}}
  }
  HorizontalBox_0 {
   widgetClassPath=&HorizontalBox
   parent=Overlay_0
   widget=HorizontalBox_0
   slot {class=OverlaySlot, properties {horizontalAlignment=HAlign_Fill, padding=<ltrb 10 10 10 10>}}
  }
  ScaleBox_0 {
   widgetClassPath=&ScaleBox
   parent=HorizontalBox_0
   widget=ScaleBox_0
   slot {class=HorizontalBoxSlot, properties {horizontalAlignment=HAlign_Left, verticalAlignment=VAlign_Center, size {sizeRule=Fill, value=1}}}
  }
  Overlay_1 {
   widgetClassPath=&Overlay
   parent=ScaleBox_0
   widget=Overlay_1
   slot {class=ScaleBoxSlot, properties {horizontalAlignment=HAlign_Fill, verticalAlignment=VAlign_Fill}}
  }
  ID_Name {
   widgetClassPath=&TextBlock
   bIsVariable=true
   parent=Overlay_1
   widget=ID_Name
   properties {
    text=ID Name
    font {
     bForceMonospaced=false
     bMaterialIsStencil=false
     fontMaterial=null
     fontObject=&Roboto
     letterSpacing=0
     monospacedWidth=1
     size=10.5
     skewAmount=0
     typefaceFontName=Bold
     outlineSettings {
      bApplyOutlineToDropShadows=false
      bMiteredCorners=false
      bSeparateFillAlpha=false
      outlineMaterial=null
      outlineSize=0
      outlineColor=<rgba 0 0 0 1>
     }
    }
   }
   slot {class=OverlaySlot, properties {verticalAlignment=VAlign_Center}}
  }
  btn_OpenID {
   widgetClassPath=&EditorUtilityButton
   parent=Overlay_1
   widget=btn_OpenID
   properties {
    renderOpacity=0
    widgetStyle {
     clickedSlateSound {resourceObject=null}
     disabled {
      bIsDynamicallyLoaded=false
      drawAs=RoundedBox
      imageType=NoImage
      mirroring=NoMirror
      resourceName=None
      resourceObject=null
      tiling=NoTile
      imageSize=<0 0>
      margin=<ltrb 0 0 0 0>
      outlineSettings {
       bUseBrushTransparency=false
       roundingType=FixedRadius
       width=1
       color {colorUseRule=UseColor_ColorTable, specifiedColor=<rgba 0 0 0 0>}
       cornerRadii=<4 4 4 4>
      }
      tintColor {colorUseRule=UseColor_Specified, specifiedColor=<rgba 0.0395 0.0395 0.0395 1>}
      uVRegion {bIsValid=false, max=<0 0>, min=<0 0>}
     }
     disabledForeground {colorUseRule=UseColor_ColorTable, specifiedColor=<rgba 0 0 0 0>}
     hovered {
      bIsDynamicallyLoaded=false
      drawAs=RoundedBox
      imageType=NoImage
      mirroring=NoMirror
      resourceName=None
      resourceObject=null
      tiling=NoTile
      imageSize=<32 32>
      margin=<ltrb 0 0 0 0>
      outlineSettings {
       bUseBrushTransparency=false
       roundingType=FixedRadius
       width=1
       color {colorUseRule=UseColor_ColorTable, specifiedColor=<rgba 0 0 0 0>}
       cornerRadii=<4 4 4 4>
      }
      tintColor {colorUseRule=UseColor_Specified, specifiedColor=<rgba 0.0953 0.0953 0.0953 1>}
      uVRegion {bIsValid=false, max=<0 0>, min=<0 0>}
     }
     hoveredForeground {colorUseRule=UseColor_ColorTable, specifiedColor=<rgba 0 0 0 0>}
     hoveredSlateSound {resourceObject=null}
     normal {
      bIsDynamicallyLoaded=false
      drawAs=NoDrawType
      imageType=NoImage
      mirroring=NoMirror
      resourceName=None
      resourceObject=null
      tiling=NoTile
      imageSize=<32 32>
      margin=<ltrb 0 0 0 0>
      outlineSettings {
       bUseBrushTransparency=false
       roundingType=FixedRadius
       width=1
       color {colorUseRule=UseColor_ColorTable, specifiedColor=<rgba 0 0 0 0>}
       cornerRadii=<4 4 4 4>
      }
      tintColor {colorUseRule=UseColor_Specified, specifiedColor=<rgba 0.0395 0.0395 0.0395 1>}
      uVRegion {bIsValid=false, max=<0 0>, min=<0 0>}
     }
     normalForeground {colorUseRule=UseColor_ColorTable, specifiedColor=<rgba 0 0 0 0>}
     normalPadding=<ltrb 12 1.5 12 1.5>
     pressed {
      bIsDynamicallyLoaded=false
      drawAs=RoundedBox
      imageType=NoImage
      mirroring=NoMirror
      resourceName=None
      resourceObject=null
      tiling=NoTile
      imageSize=<32 32>
      margin=<ltrb 0 0 0 0>
      outlineSettings {
       bUseBrushTransparency=false
       roundingType=FixedRadius
       width=1
       color {colorUseRule=UseColor_ColorTable, specifiedColor=<rgba 0 0 0 0>}
       cornerRadii=<4 4 4 4>
      }
      tintColor {colorUseRule=UseColor_Specified, specifiedColor=<rgba 0.0284 0.0284 0.0284 1>}
      uVRegion {bIsValid=false, max=<0 0>, min=<0 0>}
     }
     pressedForeground {colorUseRule=UseColor_ColorTable, specifiedColor=<rgba 0 0 0 0>}
     pressedPadding=<ltrb 12 2.5 12 0.5>
     pressedSlateSound {resourceObject=null}
    }
   }
   slot {class=OverlaySlot, properties {horizontalAlignment=HAlign_Fill}}
  }
  Spacer_79 {widgetClassPath=&Spacer, parent=HorizontalBox_0, widget=Spacer_79, properties {size=<10 1>}, slot {class=HorizontalBoxSlot}}
  BTN_Reload {
   widgetClassPath=&EditorUtilityButton
   parent=HorizontalBox_0
   widget=BTN_Reload
   properties {widgetStyle {
     clickedSlateSound {resourceObject=null}
     disabled {
      bIsDynamicallyLoaded=false
      drawAs=Image
      imageType=FullColor
      mirroring=NoMirror
      resourceName=None
      resourceObject=/FortUILibrary/Textures/Icons/T_UI_Icon_UIKit_Update
      tiling=NoTile
      imageSize=<64 64>
      margin=<ltrb 0 0 0 0>
      outlineSettings {
       bUseBrushTransparency=false
       roundingType=FixedRadius
       width=1
       color {colorUseRule=UseColor_ColorTable, specifiedColor=<rgba 0 0 0 0>}
       cornerRadii=<4 4 4 4>
      }
      tintColor {colorUseRule=UseColor_Specified, specifiedColor=<rgba 0.1865 0.1865 0.1865 1>}
      uVRegion {bIsValid=false, max=<0 0>, min=<0 0>}
     }
     disabledForeground {colorUseRule=UseColor_ColorTable, specifiedColor=<rgba 0 0 0 0>}
     hovered {
      bIsDynamicallyLoaded=false
      drawAs=Image
      imageType=FullColor
      mirroring=NoMirror
      resourceName=None
      resourceObject=&T_UI_Icon_UIKit_Update
      tiling=NoTile
      imageSize=<64 64>
      margin=<ltrb 0 0 0 0>
      outlineSettings {
       bUseBrushTransparency=false
       roundingType=FixedRadius
       width=1
       color {colorUseRule=UseColor_ColorTable, specifiedColor=<rgba 0 0 0 0>}
       cornerRadii=<4 4 4 4>
      }
      tintColor {colorUseRule=UseColor_Specified, specifiedColor=<rgba 0.4698 0.4698 0.4698 1>}
      uVRegion {bIsValid=false, max=<0 0>, min=<0 0>}
     }
     hoveredForeground {colorUseRule=UseColor_ColorTable, specifiedColor=<rgba 0 0 0 0>}
     hoveredSlateSound {resourceObject=null}
     normal {
      bIsDynamicallyLoaded=false
      drawAs=Image
      imageType=FullColor
      mirroring=NoMirror
      resourceName=None
      resourceObject=&T_UI_Icon_UIKit_Update
      tiling=NoTile
      imageSize=<64 64>
      margin=<ltrb 0 0 0 0>
      outlineSettings {
       bUseBrushTransparency=false
       roundingType=FixedRadius
       width=1
       color {colorUseRule=UseColor_ColorTable, specifiedColor=<rgba 0 0 0 0>}
       cornerRadii=<4 4 4 4>
      }
      tintColor {colorUseRule=UseColor_Specified, specifiedColor=<rgba 0.1865 0.1865 0.1865 1>}
      uVRegion {bIsValid=false, max=<0 0>, min=<0 0>}
     }
     normalForeground {colorUseRule=UseColor_ColorTable, specifiedColor=<rgba 0 0 0 0>}
     normalPadding=<ltrb 12 1.5 12 1.5>
     pressed {
      bIsDynamicallyLoaded=false
      drawAs=Image
      imageType=FullColor
      mirroring=NoMirror
      resourceName=None
      resourceObject=&T_UI_Icon_UIKit_Update
      tiling=NoTile
      imageSize=<64 64>
      margin=<ltrb 0 0 0 0>
      outlineSettings {
       bUseBrushTransparency=false
       roundingType=FixedRadius
       width=1
       color {colorUseRule=UseColor_ColorTable, specifiedColor=<rgba 0 0 0 0>}
       cornerRadii=<4 4 4 4>
      }
      tintColor {colorUseRule=UseColor_Specified, specifiedColor=<rgba 0.1865 0.1865 0.1865 1>}
      uVRegion {bIsValid=false, max=<0 0>, min=<0 0>}
     }
     pressedForeground {colorUseRule=UseColor_ColorTable, specifiedColor=<rgba 0 0 0 0>}
     pressedPadding=<ltrb 12 2.5 12 0.5>
     pressedSlateSound {resourceObject=null}
    }}
   slot {class=HorizontalBoxSlot}
  }
  Spacer_1 {widgetClassPath=&Spacer, parent=HorizontalBox_0, widget=Spacer_1, properties {size=<10 1>}, slot {class=HorizontalBoxSlot}}
  BTN_LoadID {widgetClassPath=&EditorUtilityButton, parent=HorizontalBox_0, widget=BTN_LoadID, slot {class=HorizontalBoxSlot}}
  ScaleBox_1 {
   widgetClassPath=&ScaleBox
   parent=BTN_LoadID
   widget=ScaleBox_1
   slot {class=ButtonSlot, properties {horizontalAlignment=HAlign_Fill, verticalAlignment=VAlign_Fill, padding=<ltrb 0 0 0 0>}}
  }
  txt_Load {
   widgetClassPath=&TextBlock
   parent=ScaleBox_1
   widget=txt_Load
   properties {
    justification=Center
    text=Load ID
    font {
     bForceMonospaced=false
     bMaterialIsStencil=false
     fontMaterial=null
     fontObject=&Roboto
     letterSpacing=0
     monospacedWidth=1
     size=9
     skewAmount=0
     typefaceFontName=Bold
     outlineSettings {
      bApplyOutlineToDropShadows=false
      bMiteredCorners=false
      bSeparateFillAlpha=false
      outlineMaterial=null
      outlineSize=0
      outlineColor=<rgba 0 0 0 1>
     }
    }
   }
   slot {class=ScaleBoxSlot, properties {horizontalAlignment=HAlign_Fill}}
  }
  Spacer {widgetClassPath=&Spacer, parent=HorizontalBox_0, widget=Spacer, properties {size=<10 1>}, slot {class=HorizontalBoxSlot}}
  BTN_RemoveFromList {
   widgetClassPath=&EditorUtilityButton
   parent=HorizontalBox_0
   widget=BTN_RemoveFromList
   properties {widgetStyle {
     clickedSlateSound {resourceObject=null}
     disabled {
      bIsDynamicallyLoaded=false
      drawAs=RoundedBox
      imageType=NoImage
      mirroring=NoMirror
      resourceName=None
      resourceObject=null
      tiling=NoTile
      imageSize=<0 0>
      margin=<ltrb 0 0 0 0>
      outlineSettings {
       bUseBrushTransparency=false
       roundingType=FixedRadius
       width=1
       color {colorUseRule=UseColor_ColorTable, specifiedColor=<rgba 0 0 0 0>}
       cornerRadii=<4 4 4 4>
      }
      tintColor {colorUseRule=UseColor_Specified, specifiedColor=<rgba 0.0395 0.0395 0.0395 1>}
      uVRegion {bIsValid=false, max=<0 0>, min=<0 0>}
     }
     disabledForeground {colorUseRule=UseColor_ColorTable, specifiedColor=<rgba 0 0 0 0>}
     hovered {
      bIsDynamicallyLoaded=false
      drawAs=RoundedBox
      imageType=NoImage
      mirroring=NoMirror
      resourceName=None
      resourceObject=null
      tiling=NoTile
      imageSize=<32 32>
      margin=<ltrb 0 0 0 0>
      outlineSettings {
       bUseBrushTransparency=false
       roundingType=FixedRadius
       width=1
       color {colorUseRule=UseColor_ColorTable, specifiedColor=<rgba 0 0 0 0>}
       cornerRadii=<4 4 4 4>
      }
      tintColor {colorUseRule=UseColor_Specified, specifiedColor=<rgba 0.0953 0.0953 0.0953 1>}
      uVRegion {bIsValid=false, max=<0 0>, min=<0 0>}
     }
     hoveredForeground {colorUseRule=UseColor_ColorTable, specifiedColor=<rgba 0 0 0 0>}
     hoveredSlateSound {resourceObject=null}
     normal {
      bIsDynamicallyLoaded=false
      drawAs=RoundedBox
      imageType=NoImage
      mirroring=NoMirror
      resourceName=None
      resourceObject=null
      tiling=NoTile
      imageSize=<32 32>
      margin=<ltrb 0 0 0 0>
      outlineSettings {
       bUseBrushTransparency=false
       roundingType=FixedRadius
       width=1
       color {colorUseRule=UseColor_ColorTable, specifiedColor=<rgba 0 0 0 0>}
       cornerRadii=<4 4 4 4>
      }
      tintColor {colorUseRule=UseColor_Specified, specifiedColor=<rgba 0.087 0.0005 0.0005 1>}
      uVRegion {bIsValid=false, max=<0 0>, min=<0 0>}
     }
     normalForeground {colorUseRule=UseColor_ColorTable, specifiedColor=<rgba 0 0 0 0>}
     normalPadding=<ltrb 12 1.5 12 1.5>
     pressed {
      bIsDynamicallyLoaded=false
      drawAs=RoundedBox
      imageType=NoImage
      mirroring=NoMirror
      resourceName=None
      resourceObject=null
      tiling=NoTile
      imageSize=<32 32>
      margin=<ltrb 0 0 0 0>
      outlineSettings {
       bUseBrushTransparency=false
       roundingType=FixedRadius
       width=1
       color {colorUseRule=UseColor_ColorTable, specifiedColor=<rgba 0 0 0 0>}
       cornerRadii=<4 4 4 4>
      }
      tintColor {colorUseRule=UseColor_Specified, specifiedColor=<rgba 0.0284 0.0284 0.0284 1>}
      uVRegion {bIsValid=false, max=<0 0>, min=<0 0>}
     }
     pressedForeground {colorUseRule=UseColor_ColorTable, specifiedColor=<rgba 0 0 0 0>}
     pressedPadding=<ltrb 12 2.5 12 0.5>
     pressedSlateSound {resourceObject=null}
    }}
   slot {class=HorizontalBoxSlot}
  }
  ScaleBox_2 {
   widgetClassPath=&ScaleBox
   parent=BTN_RemoveFromList
   widget=ScaleBox_2
   slot {class=ButtonSlot, properties {horizontalAlignment=HAlign_Fill, verticalAlignment=VAlign_Fill, padding=<ltrb 0 0 0 0>}}
  }
  txt_RemoveFromList {
   widgetClassPath=&TextBlock
   parent=ScaleBox_2
   widget=txt_RemoveFromList
   properties {
    justification=Center
    text=Remove
    font {
     bForceMonospaced=false
     bMaterialIsStencil=false
     fontMaterial=null
     fontObject=&Roboto
     letterSpacing=0
     monospacedWidth=1
     size=9
     skewAmount=0
     typefaceFontName=Bold
     outlineSettings {
      bApplyOutlineToDropShadows=false
      bMiteredCorners=false
      bSeparateFillAlpha=false
      outlineMaterial=null
      outlineSize=0
      outlineColor=<rgba 0 0 0 1>
     }
    }
   }
   slot {class=ScaleBoxSlot, properties {horizontalAlignment=HAlign_Fill}}
  }
 ]
}
